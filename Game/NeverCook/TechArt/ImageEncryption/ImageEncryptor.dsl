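; Editor tool actor that encrypts a texture through a material and previews its decryption per mip level.
; Construction script builds encrypt and live-decrypt materials, plus nine labelled mip columns comparing static decryption with the original.
; Two custom events draw encryption into EncryptorRT and bake static textures to local-test or in-game paths, setting their LOD group.
; The local test also feeds the mip previews and appends each mip's resolution to labels.
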(event EventBeginPlay)

(event Collision|EventActorBeginOverlap
  (params (OtherActor :type Actor%20Object%20Reference)))

(event EventTick
  (params (DeltaSeconds :type Float%20%28single-precision%29)))

(event Custom|1%20Encrypt%20Image%20Local%20Test%20and%20Debug
  (bind _statictexture (Variables|Default|GetStaticTexture))
  (Rendering|DrawMaterialtoRenderTarget "/Game/NeverCook/TechArt/ImageEncryption/EncryptorRT.EncryptorRT" (Variables|Default|GetEncyptedMaterial))
  (bind _returnvalue (EditorScripting|Rendering|RenderTarget2DCreateStaticTexture2DEditorOnly "/Game/NeverCook/TechArt/ImageEncryption/EncryptorRT.EncryptorRT" (Utilities|String|Append (Variables|Default|GetLocal%20Test%20Location) (Variables|Default|GetName))))
  (Variables|/Script/Engine.Texture|SetLODGroup "TEXTUREGROUP_Project06" _returnvalue)
  (Variables|Default|SetStaticTexture _returnvalue)
  (for (_array_element _array_index) (Variables|Default|GetMipMaterialList)
    (Rendering|Material|SetTextureParameterValue _array_element "Image" _statictexture)
    (Rendering|Material|SetScalarParameterValue _array_element "MipLevel" (Math|Conversions|ToFloat(Integer) _array_index) :!class /Script/Engine.MaterialInstanceDynamic)
    (Rendering|Components|TextRender|SetText (Utilities|Array|Get(acopy) (Variables|Default|GetTextList) _array_index) (Utilities|Text|ToText(String) (Utilities|String|Append (Utilities|String|Append (Utilities|String|ToString(Text) (Variables|/Script/Engine.TextRenderComponent|GetText (Utilities|Array|Get(acopy) (Variables|Default|GetTextList) _array_index))) " ") (Utilities|String|Append (Utilities|String|Append (Utilities|String|ToString(Integer) (/ (Rendering|Texture|GetSizeX _statictexture) (Math|Float|Truncate (Math|Float|Power 2.0 (Math|Conversions|ToFloat(Integer) _array_index))))) "x") (Utilities|String|ToString(Integer) (/ (Rendering|Texture|GetSizeY _statictexture) (Math|Float|Truncate (Math|Float|Power 2.0 (Math|Conversions|ToFloat(Integer) _array_index)))))))))))

(event Custom|2%20Encrypt%20Image%20and%20Save%20inGame
  (Rendering|DrawMaterialtoRenderTarget "/Game/NeverCook/TechArt/ImageEncryption/EncryptorRT.EncryptorRT" (Variables|Default|GetEncyptedMaterial))
  (bind _returnvalue (EditorScripting|Rendering|RenderTarget2DCreateStaticTexture2DEditorOnly "/Game/NeverCook/TechArt/ImageEncryption/EncryptorRT.EncryptorRT" (Utilities|String|Append (Variables|Default|GetLocation_InGameTexture) (Variables|Default|GetName))))
  (Variables|/Script/Engine.Texture|SetLODGroup "TEXTUREGROUP_Project06" _returnvalue)
  (Variables|Default|SetStaticTexture_inGame _returnvalue))

(fn ConstructionScript
  (bind _returnvalue (Math|Vector|MakeVector 1922.0))
  (Variables|Default|SetName (Utilities|GetDisplayName (Variables|Default|GetTexture)))
  (bind _returnvalue_2 (Rendering|Material|CreateDynamicMaterialInstance (Variables|Default|GetSM_Plane_Encryption) 0 "/Game/NeverCook/TechArt/ImageEncryption/M_Encryption.M_Encryption" "None" :!class /Script/Engine.PrimitiveComponent))
  (bind _output_get (Variables|Default|SetEncyptedMaterial _returnvalue_2))
  (Rendering|Material|SetTextureParameterValue _output_get "Image" (Variables|Default|GetTexture))
  (bind _returnvalue_3 (Rendering|Material|CreateDynamicMaterialInstance (Variables|Default|GetSM_Plane_Decrypted) 0 "/Game/NeverCook/TechArt/ImageEncryption/M_Decryption.M_Decryption" "None" :!class /Script/Engine.PrimitiveComponent))
  (Rendering|Material|SetTextureParameterValue _returnvalue_3 "Image" "/Game/NeverCook/TechArt/ImageEncryption/EncryptorRT.EncryptorRT")
  (for _index (range 9)
    (Utilities|FlowControl|Sequence
      (:then_0
        (bind _returnvalue_4 (AddComponent|NODE_AddTextRenderComponent-0 :RelativeTransform (Math|Transform|MakeTransform (* _returnvalue (Math|Conversions|ToFloat(Integer) (+ _index 2))) "0, 0, 0" "17.695000,9.982500,9.982500")))
        (Transformation|AddRelativeLocation _returnvalue_4 "0, 0,500.000000" false false)
        (Transformation|AddRelativeRotation _returnvalue_4 "0,90.000000, 0" false false)
        (Rendering|Components|TextRender|SetText _returnvalue_4 (Utilities|Text|ToText(String) (Utilities|String|Append "MipLevel " (Utilities|String|ToString(Integer) _index))))
        (Utilities|Array|Add (Variables|Default|GetTextList) _returnvalue_4))
      (:then_1
        (bind _returnvalue_5 (AddComponent|NODE_AddStaticMeshComponent-0 :RelativeTransform (Math|Transform|MakeTransform (* _returnvalue (Math|Conversions|ToFloat(Integer) (+ _index 2))) "0, 0, 0" "17.695000,9.982500,9.982500")))
        (bind _returnvalue_6 (Rendering|Material|CreateDynamicMaterialInstance _returnvalue_5 0 "/Game/NeverCook/TechArt/ImageEncryption/M_Decryption_Static.M_Decryption_Static" "None" :!class /Script/Engine.PrimitiveComponent))
        (Utilities|Array|Add (Variables|Default|GetMipMaterialList) _returnvalue_6))
      (:then_2
        (bind _returnvalue_7 (AddComponent|NODE_AddStaticMeshComponent-1 :RelativeTransform (Math|Transform|MakeTransform (+ (* _returnvalue (Math|Conversions|ToFloat(Integer) (+ _index 2))) "0,0.000000,1200.000000") "0, 0, 0" "17.695000,9.982500,9.982500")))
        (bind _returnvalue_8 (Rendering|Material|CreateDynamicMaterialInstance _returnvalue_7 0 "/Game/NeverCook/TechArt/ImageEncryption/M_Decryption_Original.M_Decryption_Original" "None" :!class /Script/Engine.PrimitiveComponent))
        (Rendering|Material|SetTextureParameterValue _returnvalue_8 "Image" (Variables|Default|GetTexture))
        (Rendering|Material|SetScalarParameterValue _returnvalue_8 "MipLevel" (Math|Conversions|ToFloat(Integer) _index) :!class /Script/Engine.MaterialInstanceDynamic)))))
