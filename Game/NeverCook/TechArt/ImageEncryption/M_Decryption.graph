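# M_Decryption is a tech-art material for image decryption that switches between an original texture and a processed version.
# An Image texture parameter and a MipLevel scalar feed both branches of the Original static switch.
# The true branch samples raw RGB at the explicit mip level.
# The false branch passes both inputs and a DebugMips static bool through the shared MF_Magic function, whose internals live outside this material.
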Material /Game/NeverCook/TechArt/ImageEncryption/M_Decryption
graph {
 properties {pixelDepthOffsetMode=PDOM_Legacy, usedShadingModels=Default Lit}
 nodes [
  n1 = TextureObjectParameter() {parameterName=Image}
  n2 = ScalarParameter() {parameterName=MipLevel}
  StaticSwitchParameter(
   True=TextureSample!(UVs, Tex=@n1, Level=@n2, Apply View MipBias) {mipValueMode=TMVM_MipLevel}.RGB
   False=MaterialFunctionCall!(Texture=@n1, UV, Encrypt, DebugMips=StaticBoolParameter() {parameterName=DebugMips}, MipLevel=@n2) {materialFunction=/Game/Packages/Fortress_SharedMaterials/Base_Material_Functions/MF_Magic}.Result
  ) {parameterName=Original}
 ]
}
