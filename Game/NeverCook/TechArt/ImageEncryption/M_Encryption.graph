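# Lit material in a never-cook TechArt folder that sets up image encryption on a texture.
# An `Image` texture parameter feeds the shared `MF_Magic` function with its Encrypt switch statically on, leaving UV and mip debug inputs at defaults.
# The graph shows no connections to material outputs, and a lone constant is unused, so it appears to be a tool or scratch setup.
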Material /Game/NeverCook/TechArt/ImageEncryption/M_Encryption
graph {
 properties {pixelDepthOffsetMode=PDOM_Legacy, usedShadingModels=Default Lit}
 nodes [
  Constant() {r=1}
  MaterialFunctionCall!(Texture=TextureObjectParameter() {parameterName=Image}, UV, Encrypt=StaticBool() {value=true}, DebugMips, MipLevel) {materialFunction=/Game/Packages/Fortress_SharedMaterials/Base_Material_Functions/MF_Magic}
 ]
}
